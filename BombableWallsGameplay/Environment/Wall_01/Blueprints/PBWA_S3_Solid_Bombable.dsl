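; Bombable solid wall Blueprint that destroys itself when it takes qualifying explosive damage.
; On server damage, it checks the damage tags against its explosion tags.
; Unless flagged two-sided, it also requires the hit from one side, using a dot product of its right vector and the impact-to-actor offset.
; A scalable-float hotfix bypasses both checks so any damage triggers death, tagged as explosive.
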(event EventOnDamageServer
  (params (Damage :type Float%20%28single-precision%29) (DamageTags :type Gameplay%20Tag%20Container%20Structure%20%28by%20ref%29) (Momentum :type Vector :default "0, 0, 0") (HitInfo :type Hit%20Result%20Structure%20%28by%20ref%29) (InstigatedBy :type Controller%20Object%20Reference) (DamageCauser :type Actor%20Object%20Reference) (EffectContext :type Gameplay%20Effect%20Context%20Handle%20Structure))
  (bind (_bblockinghit _binitialoverlap _time _distance _location _impactpoint) (Collision|BreakHitResult HitInfo))
  ;[0 0 1] Check for explosive damage. If wall is onesided, check for correct damage direction. Destroy wall if it all adds up
  ;;[1 0 1] Hotfix value to allow destruction by ANYTHING
  (if (ScalableFloat|GetValueasBool (Variables|Default|GetVulnerableToAllDamage_Hotfix))
    _
    (elif (and (GameplayTags|HasAnyTags DamageTags (GameplayTags|MakeGameplayTagContainerfromArray (Variables|Default|GetExplosionTags)) true) (or (Variables|Default|GetTwoSided) (>= (Math|Vector|DotProduct (Transformation|GetActorRightVector) (- (Transformation|GetActorLocation) _impactpoint)) 0)))
      _))
  (Gameplay|Die "(TagName=\"Gameplay.Damage.Physical.Explosive\")" InstigatedBy DamageCauser))

(fn ConstructionScript)
